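# Physics asset for the neck mesh of the female medium Nevermore cosmetic body, setting up physics for front and mid hair, five neck pieces and two neck straps.
# Small base_ capsules root the dyn_ capsules, whose hair and strap chains mostly widen their swing limits toward the tips under tight twist limits.
# Capsules on head, upper spine, clavicles and upper arms, plus a spine_05 box, form body volumes linked by angularly free constraints.
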PhysicsAsset /BRCosmetics/Characters/Player/Female/Medium/Bodies/F_MED_Nevermore/Meshes/F_MED_Nevermore_Neck_Physics
bodies {
 base_hair_front [{shapeType=Capsule, length=0.5, radius=0.5, shapeName=base_hair_front_capsule}]
 base_hair_mid [{shapeType=Capsule, length=0.5, radius=0.5, shapeName=base_hair_mid_capsule}]
 base_neck_l [{shapeType=Capsule, length=0.505, radius=0.505, shapeName=base_neck_l_capsule, rotation=<rot 0 -90 0>}]
 base_neck_mid [{shapeType=Capsule, length=0.505, radius=0.505, shapeName=base_neck_mid_capsule, rotation=<rot 0 -90 0>}]
 base_neck_mid_l [{shapeType=Capsule, length=0.505, radius=0.505, shapeName=base_neck_mid_l_capsule, rotation=<rot 0 -90 0>}]
 base_neck_mid_r [{shapeType=Capsule, length=0.505, radius=0.505, shapeName=base_neck_mid_r_capsule, rotation=<rot 0 -90 0>}]
 base_neck_r [{shapeType=Capsule, length=0.505, radius=0.505, shapeName=base_neck_r_capsule, rotation=<rot 0 -90 0>}]
 base_strap_neck_l [{shapeType=Capsule, length=0.505, radius=0.505, shapeName=base_strap_neck_l_capsule, rotation=<rot 0 -90 0>}]
 base_strap_neck_r [{shapeType=Capsule, length=0.505, radius=0.505, shapeName=base_strap_neck_r_capsule, rotation=<rot 0 -90 0>}]
 clavicle_l [{shapeType=Capsule, length=4, radius=6, shapeName=clavicle_l_capsule, center=<6 3 -3.5>, rotation=<rot 90 0 0>}]
 clavicle_r [{shapeType=Capsule, length=4, radius=5.5, shapeName=clavicle_r_capsule, center=<-6 -2.5 3.5>, rotation=<rot 90 0 0>}]
 dyn_hair_front_01 [{shapeType=Capsule, length=4, radius=2, shapeName=dyn_hair_front_01_capsule, center=<2.6004 0 1.85>, rotation=<rot 90 0 0>}]
 dyn_hair_front_02 [{shapeType=Capsule, length=4, radius=2, shapeName=dyn_hair_front_02_capsule, center=<2.6004 1.3 1.6>, rotation=<rot 90 0 0>}]
 dyn_hair_front_03 [{shapeType=Capsule, length=4, radius=2, shapeName=dyn_hair_front_03_capsule, center=<2.6004 1.3 0.4>, rotation=<rot 90 0 0>}]
 dyn_hair_mid_01 [{shapeType=Capsule, length=4, radius=1.8, shapeName=dyn_hair_mid_01_capsule, center=<2.5978 -0.5 1>, rotation=<rot 90 -4.9999 0>}]
 dyn_hair_mid_02 [{shapeType=Capsule, length=4, radius=1.5, shapeName=dyn_hair_mid_02_capsule, center=<1.9825 -0.5 0.85>, rotation=<rot 90 0 0>}]
 dyn_neck_l [{shapeType=Capsule, length=11, radius=4, shapeName=dyn_neck_l_capsule, center=<9.1 0 2>, rotation=<rot 90 0.0002 0>}]
 dyn_neck_mid [{shapeType=Capsule, length=9, radius=4.4, shapeName=dyn_neck_mid_capsule, center=<5.0042 0 3>, rotation=<rot 85 0 0>}]
 dyn_neck_mid_l [{shapeType=Capsule, length=9, radius=4, shapeName=dyn_neck_mid_l_capsule, center=<5.0042 0 2.5>, rotation=<rot 85 0 0>}]
 dyn_neck_mid_r [{shapeType=Capsule, length=9, radius=4.4, shapeName=dyn_neck_mid_r_capsule, center=<5.0042 0 3>, rotation=<rot 85 0 0>}]
 dyn_neck_r [{shapeType=Capsule, length=11, radius=4, shapeName=dyn_neck_r_capsule, center=<9.1004 0 3>, rotation=<rot 85 0 0>}]
 dyn_strap_neck_l_01 [{shapeType=Capsule, length=4, radius=1.2, shapeName=dyn_strap_neck_l_01_capsule, center=<2.4939 0 0>, rotation=<rot 90 0 0>}]
 dyn_strap_neck_l_02 [{shapeType=Capsule, length=4, radius=1.3, shapeName=dyn_strap_neck_l_02_capsule, center=<2.4939 0.1 0>, rotation=<rot 90 0 0>}]
 dyn_strap_neck_l_03 [{shapeType=Capsule, length=4, radius=1.3, shapeName=dyn_strap_neck_l_03_capsule, center=<2.4939 0 0>, rotation=<rot 90 0 0>}]
 dyn_strap_neck_r_01 [{shapeType=Capsule, length=4, radius=1.2, shapeName=dyn_strap_neck_r_01_capsule, center=<1.5767 0 0>, rotation=<rot 90 0 0>}]
 dyn_strap_neck_r_02 [{
   shapeType=Capsule
   length=2.5
   radius=1.2
   shapeName=dyn_strap_neck_r_02_capsule
   center=<4.5642 -0.3789 0.1028>
   rotation=<rot -0.0003 -10.0009 110.0004>
  }]
 head [{shapeType=Capsule, length=8, radius=7.8, shapeName=head_capsule, center=<3.0034 0.3919 0>, rotation=<rot 90 -25 0>}]
 spine_04 [{shapeType=Capsule, length=8, radius=8.5, shapeName=spine_04_capsule, center=<-0.4801 6 0>}]
 spine_05 [
  {shapeType=Capsule, length=11, radius=9, shapeName=spine_05_capsule, center=<-1.4711 1.1332 0>}
  {shapeType=Box, extentX=6.5, extentY=4, extentZ=6.5, shapeName=spine_05_box, center=<-0.2 9.8824 -0.2465>, rotation=<rot -47.1186 -34.2168 0>}
 ]
 upperarm_l [{shapeType=Capsule, length=22, radius=4, shapeName=upperarm_l_capsule, center=<13 0 -1.3012>, rotation=<rot 90 0 0>}]
 upperarm_r [{shapeType=Capsule, length=22, radius=2.55, shapeName=upperarm_l_capsule, center=<-13 0 1.3012>, rotation=<rot 90 0 0>}]
}
constraints [
 {bone1Name=dyn_neck_l, bone2Name=base_neck_l, swing1LimitDegrees=30, swing2LimitDegrees=40, twistLimitDegrees=5}
 {bone1Name=dyn_neck_mid, bone2Name=base_neck_mid, swing1LimitDegrees=20, swing2LimitDegrees=40, twistLimitDegrees=5}
 {bone1Name=dyn_neck_mid_l, bone2Name=base_neck_mid_l, swing1LimitDegrees=25, swing2LimitDegrees=40, twistLimitDegrees=5}
 {bone1Name=dyn_neck_mid_r, bone2Name=base_neck_mid_r, swing1LimitDegrees=20, swing2LimitDegrees=40, twistLimitDegrees=5}
 {bone1Name=dyn_neck_r, bone2Name=base_neck_r, swing1LimitDegrees=20, swing2LimitDegrees=40, twistLimitDegrees=5}
 {
  bone1Name=upperarm_r
  bone2Name=clavicle_r
  swing1LimitDegrees=45
  swing1Motion=Free
  swing2LimitDegrees=45
  swing2Motion=Free
  twistLimitDegrees=45
  twistMotion=Free
 }
 {
  bone1Name=spine_04
  bone2Name=spine_01
  swing1LimitDegrees=45
  swing1Motion=Free
  swing2LimitDegrees=45
  swing2Motion=Free
  twistLimitDegrees=45
  twistMotion=Free
 }
 {
  bone1Name=spine_05
  bone2Name=spine_04
  swing1LimitDegrees=45
  swing1Motion=Free
  swing2LimitDegrees=45
  swing2Motion=Free
  twistLimitDegrees=45
  twistMotion=Free
 }
 {
  bone1Name=base_strap_neck_l
  bone2Name=spine_05
  swing1LimitDegrees=45
  swing1Motion=Free
  swing2LimitDegrees=45
  swing2Motion=Free
  twistLimitDegrees=45
  twistMotion=Free
 }
 {bone1Name=dyn_strap_neck_l_01, bone2Name=base_strap_neck_l, swing1LimitDegrees=15, swing2LimitDegrees=8, twistLimitDegrees=5}
 {bone1Name=dyn_strap_neck_l_02, bone2Name=dyn_strap_neck_l_01, swing1LimitDegrees=18, swing2LimitDegrees=10, twistLimitDegrees=5}
 {bone1Name=dyn_strap_neck_l_03, bone2Name=dyn_strap_neck_l_02, swing1LimitDegrees=20, swing2LimitDegrees=10, twistLimitDegrees=5}
 {
  bone1Name=base_strap_neck_r
  bone2Name=spine_05
  swing1LimitDegrees=45
  swing1Motion=Free
  swing2LimitDegrees=45
  swing2Motion=Free
  twistLimitDegrees=45
  twistMotion=Free
 }
 {bone1Name=dyn_strap_neck_r_01, bone2Name=base_strap_neck_r, swing1LimitDegrees=20, swing2LimitDegrees=10, twistLimitDegrees=5}
 {bone1Name=dyn_strap_neck_r_02, bone2Name=dyn_strap_neck_r_01, swing1LimitDegrees=25, swing2LimitDegrees=15, twistLimitDegrees=5}
 {
  bone1Name=clavicle_l
  bone2Name=spine_05
  swing1LimitDegrees=45
  swing1Motion=Free
  swing2LimitDegrees=45
  swing2Motion=Free
  twistLimitDegrees=45
  twistMotion=Free
 }
 {
  bone1Name=upperarm_l
  bone2Name=clavicle_l
  swing1LimitDegrees=45
  swing1Motion=Free
  swing2LimitDegrees=45
  swing2Motion=Free
  twistLimitDegrees=45
  twistMotion=Free
 }
 {
  bone1Name=clavicle_r
  bone2Name=spine_05
  swing1LimitDegrees=45
  swing1Motion=Free
  swing2LimitDegrees=45
  swing2Motion=Free
  twistLimitDegrees=45
  twistMotion=Free
 }
 {
  bone1Name=base_hair_front
  bone2Name=head
  swing1LimitDegrees=45
  swing1Motion=Free
  swing2LimitDegrees=45
  swing2Motion=Free
  twistLimitDegrees=45
  twistMotion=Free
 }
 {bone1Name=dyn_hair_front_01, bone2Name=base_hair_front, swing1LimitDegrees=10, swing2LimitDegrees=10, twistLimitDegrees=1}
 {bone1Name=dyn_hair_front_02, bone2Name=dyn_hair_front_01, swing1LimitDegrees=15, swing2LimitDegrees=20, twistLimitDegrees=1}
 {bone1Name=dyn_hair_front_03, bone2Name=dyn_hair_front_02, swing1LimitDegrees=20, swing2LimitDegrees=25, twistLimitDegrees=1}
 {
  bone1Name=base_hair_mid
  bone2Name=head
  swing1LimitDegrees=45
  swing1Motion=Free
  swing2LimitDegrees=45
  swing2Motion=Free
  twistLimitDegrees=45
  twistMotion=Free
 }
 {bone1Name=dyn_hair_mid_01, bone2Name=base_hair_mid, swing1LimitDegrees=10, swing2LimitDegrees=15, twistLimitDegrees=1}
 {bone1Name=dyn_hair_mid_02, bone2Name=dyn_hair_mid_01, swing1LimitDegrees=15, swing2LimitDegrees=20, twistLimitDegrees=1}
]
