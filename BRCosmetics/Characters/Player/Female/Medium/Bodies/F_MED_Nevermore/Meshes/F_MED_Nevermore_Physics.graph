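# Physics asset for the Nevermore female medium body cosmetic, set up for secondary motion on its left-side belt pieces.
# Pelvis, spine_01 and left thigh capsules approximate the nearby body, joined by free-rotating constraints.
# Tiny surfaceconst capsules link to spine_01 and anchor a one-bone back chain and a two-bone front chain.
# Belt joints allow a wider swing on one axis, with the other swing and twist tightly limited.
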PhysicsAsset /BRCosmetics/Characters/Player/Female/Medium/Bodies/F_MED_Nevermore/Meshes/F_MED_Nevermore_Physics
bodies {
 dyn_belt_l_back [{
   shapeType=Capsule
   length=22
   radius=2.5
   shapeName=dyn_belt_l_back_capsule
   center=<10.1345 -5.5 -0.7>
   rotation=<rot 0.4461 -107.7609 -84.8925>
  }]
 dyn_belt_l_front [{
   shapeType=Capsule
   length=14
   radius=4
   shapeName=dyn_belt_l_front_capsule
   center=<7.0797 5.5 -0.95>
   rotation=<rot -0.1693 104.345 -94.3671>
  }]
 dyn_belt_l_front_02 [{shapeType=Capsule, length=22, radius=3, shapeName=dyn_belt_l_front_02_capsule, center=<7.3263 0.8 4.2>, rotation=<rot 0 50 95>}]
 pelvis [{shapeType=Capsule, length=15, radius=9.5, shapeName=pelvis_capsule, center=<-1.0672 -2.1674 0>}]
 spine_01 [{shapeType=Capsule, length=9, radius=7.5, shapeName=spine_01_capsule}]
 surfaceconst_belt_l_back [{shapeType=Capsule, length=0.505, radius=0.505, shapeName=surfaceconst_belt_l_back_capsule, rotation=<rot 0 -90 0>}]
 surfaceconst_belt_l_front [{shapeType=Capsule, length=0.505, radius=0.505, shapeName=surfaceconst_belt_l_front_capsule, rotation=<rot 0 -90 0>}]
 thigh_l [{shapeType=Capsule, length=26, radius=3.5, shapeName=thigh_r_capsule, center=<-13.414 0 -0.0753>, rotation=<rot 97 0 0>}]
}
constraints [
 {
  bone1Name=thigh_l
  bone2Name=pelvis
  swing1LimitDegrees=45
  swing1Motion=Free
  swing2LimitDegrees=45
  swing2Motion=Free
  twistLimitDegrees=45
  twistMotion=Free
 }
 {
  bone1Name=spine_01
  bone2Name=pelvis
  swing1LimitDegrees=45
  swing1Motion=Free
  swing2LimitDegrees=45
  swing2Motion=Free
  twistLimitDegrees=45
  twistMotion=Free
 }
 {
  bone1Name=surfaceconst_belt_l_back
  bone2Name=spine_01
  swing1LimitDegrees=45
  swing1Motion=Free
  swing2LimitDegrees=45
  swing2Motion=Free
  twistLimitDegrees=45
  twistMotion=Free
 }
 {bone1Name=dyn_belt_l_back, bone2Name=surfaceconst_belt_l_back, swing1LimitDegrees=5, swing2LimitDegrees=30, twistLimitDegrees=1}
 {
  bone1Name=surfaceconst_belt_l_front
  bone2Name=spine_01
  swing1LimitDegrees=45
  swing1Motion=Free
  swing2LimitDegrees=45
  swing2Motion=Free
  twistLimitDegrees=45
  twistMotion=Free
 }
 {bone1Name=dyn_belt_l_front, bone2Name=surfaceconst_belt_l_front, swing1LimitDegrees=5, swing2LimitDegrees=20, twistLimitDegrees=1}
 {bone1Name=dyn_belt_l_front_02, bone2Name=dyn_belt_l_front, swing1LimitDegrees=1, swing2LimitDegrees=30, twistLimitDegrees=1}
]
