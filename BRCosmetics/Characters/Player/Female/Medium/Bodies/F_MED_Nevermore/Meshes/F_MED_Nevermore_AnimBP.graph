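# Post-process-style AnimBlueprint for the Nevermore female player body cosmetic, adding secondary motion to feathers, belts, straps and knife.
# Copies the parent mesh pose, layering in hair and jaw bones from face-accessory and head meshes, then runs a Control Rig.
# Chains AnimDynamics, two RigidBody sims and additive belt offsets, with spherical limits on leg straps and front-end-specific knife tuning.
# EventGraph caches part meshes and updates gravity override, front-end and LOD state.
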AnimBlueprint /BRCosmetics/Characters/Player/Female/Medium/Bodies/F_MED_Nevermore/Meshes/F_MED_Nevermore_AnimBP
target_skeleton=&/F_MED_Nevermore_Skeleton
graphs [
 AnimGraph {nodes [
   n1 = Get bIsInFrontEnd()
   n2 = Control Rig!(
    Source=Use cached pose 'StartPose'()
    Alpha
    IsCrouching:bool=Get bIsCrouching()
    PlayingMeleeAnim:bool=Get bIsPlayingMeleeAnim()
    LOD1Plus:bool=Get LOD1Plus() -> (LOD1Plus:bool)
   ) -> (Pose) {
    alphaBoolBlend {blendOption=Linear}
    controlRigAssetReference {blueprintRigClass=&/F_MED_Nevermore_CtrlRig.F_MED_Nevermore_CtrlRig_C}
   }
   n3 = Anim Dynamics - Bone: dyn_frather_wrist_r!(
    SimulationSpace:AnimPhysSimSpaceType Enum
    ComponentPose=Local To Component(LocalPose=@n2)
    Alpha=1
    GravityOverride:Vector="0,0,0"
    ComponentLinearVelScale:Vector="0,0,0"
    ExternalForce="0,0,0"
   ) -> (Pose) {
    angularDampingOverride=1
    angularSpringConstant=11000
    bAngularSpring=true
    bOverrideAngularDamping=true
    bUseGravityOverride=true
    gravityScale=0.35
    simulationSpace=RootRelative
    alphaBoolBlend {blendOption=Linear}
    boundBone {boneName=dyn_frather_wrist_r}
    componentAppliedLinearAccClamp=<9500 9500 9500>
    componentLinearAccScale=<1 1 1>
    componentLinearVelScale=<1 1 1>
    gravityOverride {z=-980}
    physicsBodyDefinitions [{
      boundBone {boneName=dyn_frather_wrist_r}
      boxExtents=<30 30 30>
      constraintSetup {angularLimitsMax=<0 3 10>, angularLimitsMin=<0 -5 -10>, angularTarget=<1 0 0>}
      localJointOffset=<2 0 0>
     }]
   }
   n4 = Anim Dynamics - Bone: dyn_frather_elbow_r!(
    SimulationSpace:AnimPhysSimSpaceType Enum
    ComponentPose=@n3
    Alpha=1
    GravityOverride:Vector="0,0,0"
    ComponentLinearVelScale:Vector="0,0,0"
    ExternalForce="0,0,0"
   ) -> (Pose) {
    angularDampingOverride=1
    angularSpringConstant=11000
    bAngularSpring=true
    bOverrideAngularDamping=true
    bUseGravityOverride=true
    gravityScale=0.35
    simulationSpace=RootRelative
    alphaBoolBlend {blendOption=Linear}
    boundBone {boneName=dyn_frather_elbow_r}
    componentAppliedLinearAccClamp=<9500 9500 9500>
    componentLinearAccScale=<1 1 1>
    componentLinearVelScale=<1 1 1>
    gravityOverride {z=-980}
    physicsBodyDefinitions [{
      boundBone {boneName=dyn_frather_elbow_r}
      boxExtents=<30 30 30>
      constraintSetup {angularLimitsMax=<3 10 10>, angularLimitsMin=<-3 -10 -10>, angularTarget=<1 0 0>}
      localJointOffset=<2 0 0>
     }]
   }
   n5 = Anim Dynamics - Bone: dyn_frather_shoulder_r!(
    SimulationSpace:AnimPhysSimSpaceType Enum
    ComponentPose=@n4
    Alpha=0.9
    GravityOverride:Vector="0,0,0"
    ComponentLinearVelScale:Vector="0,0,0"
    ExternalForce="0,0,0"
   ) -> (Pose) {
    angularDampingOverride=1
    angularSpringConstant=11000
    bAngularSpring=true
    bOverrideAngularDamping=true
    bUseGravityOverride=true
    gravityScale=0.35
    simulationSpace=RootRelative
    alphaBoolBlend {blendOption=Linear}
    boundBone {boneName=dyn_frather_shoulder_r}
    componentAppliedLinearAccClamp=<9500 9500 9500>
    componentLinearAccScale=<1 1 1>
    componentLinearVelScale=<1 1 1>
    gravityOverride {z=-980}
    physicsBodyDefinitions [{
      boundBone {boneName=dyn_frather_shoulder_r}
      boxExtents=<40 40 40>
      constraintSetup {angularLimitsMax=<5 5 15>, angularLimitsMin=<-5 -15 -15>, angularTarget=<1 0 0>}
      localJointOffset=<2 0 0>
     }]
   }
   n6 = Anim Dynamics - Bone: dyn_belt_r!(
    SimulationSpace:AnimPhysSimSpaceType Enum
    ComponentPose=@n5
    bAlphaBoolEnabled:bool=NOT Boolean(A=Get bIsPlayingMeleeAnim())
    GravityOverride:Vector="0,0,0"
    ComponentLinearVelScale:Vector="0,0,0"
    ExternalForce="0,0,0"
   ) -> (Pose) {
    alphaInputType=Bool
    angularDampingOverride=1
    angularSpringConstant=11000
    bAngularSpring=true
    bOverrideAngularDamping=true
    bUseGravityOverride=true
    gravityScale=0.35
    simulationSpace=RootRelative
    alphaBoolBlend {blendInTime=0.5, blendOption=QuadraticInOut, blendOutTime=1}
    boundBone {boneName=dyn_belt_r}
    componentAppliedLinearAccClamp=<9500 9500 9500>
    componentLinearAccScale=<1 1 1>
    componentLinearVelScale=<1 1 1>
    gravityOverride {z=-980}
    physicsBodyDefinitions [{
      boundBone {boneName=dyn_belt_r}
      boxExtents=<40 40 40>
      constraintSetup {angularLimitsMax=<5 1 5>, angularLimitsMin=<-5 -30 -5>, angularTarget=<1 0 0>}
      localJointOffset=<3 0 0>
     }]
   }
   n7 = Anim Dynamics - Bone: dyn_knife!(ComponentPose=@n6, bAlphaBoolEnabled:bool=@n1, GravityOverride:Vector="0,0,0", ComponentLinearVelScale:Vector="0,0,0", ExternalForce="0,0,0") -> (Pose) {
    alphaInputType=Bool
    angularDampingOverride=1
    angularSpringConstant=11000
    bAngularSpring=true
    bOverrideAngularDamping=true
    bUseGravityOverride=true
    gravityScale=0.35
    simulationSpace=RootRelative
    alphaBoolBlend {blendInTime=0.5, blendOption=QuadraticInOut, blendOutTime=1}
    boundBone {boneName=dyn_knife}
    componentAppliedLinearAccClamp=<9500 9500 9500>
    componentLinearAccScale=<1 1 1>
    componentLinearVelScale=<1 1 1>
    gravityOverride {z=-980}
    physicsBodyDefinitions [{
      boundBone {boneName=dyn_knife}
      boxExtents=<20 30 30>
      constraintSetup {angularLimitsMax=<1.5 4 11>, angularLimitsMin=<-1.5 -4 -11>, angularTarget=<1 0 0>}
      localJointOffset=<4 0 0>
     }]
    simSpaceSettings {maxAngularAcceleration=35, maxAngularVelocity=7, simSpaceAngularAlpha=0.4}
   }
   n8 = Anim Dynamics - Bone: dyn_knife!(
    SimulationSpace:AnimPhysSimSpaceType Enum
    ComponentPose=@n7
    bAlphaBoolEnabled=NOT Boolean(A=@n1)
    GravityOverride="0,0,0"
    ComponentLinearVelScale="0,0,0"
    ExternalForce="0,0,0"
   ) -> (Pose) {
    alphaInputType=Bool
    angularDampingOverride=1
    angularSpringConstant=11000
    bAngularSpring=true
    bOverrideAngularDamping=true
    bUseGravityOverride=true
    gravityScale=0.35
    simulationSpace=RootRelative
    alphaBoolBlend {blendInTime=0.5, blendOption=QuadraticInOut, blendOutTime=1}
    boundBone {boneName=dyn_knife}
    componentAppliedLinearAccClamp=<9500 9500 9500>
    componentLinearAccScale=<1 1 1>
    componentLinearVelScale=<1 1 1>
    gravityOverride {z=-980}
    physicsBodyDefinitions [{
      boundBone {boneName=dyn_knife}
      boxExtents=<20 30 30>
      constraintSetup {angularLimitsMax=<10 5 15>, angularLimitsMin=<-10 -5 -15>, angularTarget=<1 0 0>}
      localJointOffset=<4 0 0>
     }]
   }
   n9 = Anim Dynamics - Bone: dyn_strap_leg_r_01!(
    SimulationSpace:AnimPhysSimSpaceType Enum
    ComponentPose=@n8
    Alpha=0.9
    GravityOverride:Vector="0,0,0"
    ComponentLinearVelScale:Vector="0,0,0"
    ExternalForce="0,0,0"
   ) -> (Pose) {
    angularDampingOverride=1
    angularSpringConstant=11000
    bAngularSpring=true
    bOverrideAngularDamping=true
    bUseGravityOverride=true
    bUseSphericalLimits=true
    gravityScale=0.35
    simulationSpace=RootRelative
    alphaBoolBlend {blendOption=Linear}
    boundBone {boneName=dyn_strap_leg_r_01}
    componentAppliedLinearAccClamp=<9500 9500 9500>
    componentLinearAccScale=<1 1 1>
    componentLinearVelScale=<1 1 1>
    gravityOverride {z=-980}
    physicsBodyDefinitions [{
      collisionType=CustomSphere
      sphereCollisionRadius=1
      boundBone {boneName=dyn_strap_leg_r_01}
      boxExtents=<20 30 30>
      constraintSetup {angularLimitsMax=<0 1 15>, angularLimitsMin=<0 -10 -15>, angularTarget=<1 0 0>}
      localJointOffset=<4 0 0>
     }]
    sphericalLimits [{limitRadius=5, drivingBone {boneName=calf_r}, sphereLocalOffset=<5 1 0>}]
   }
   n10 = Anim Dynamics - Bone: dyn_strap_leg_r_02!(
    SimulationSpace:AnimPhysSimSpaceType Enum
    ComponentPose=@n9
    Alpha=0.9
    GravityOverride:Vector="0,0,0"
    ComponentLinearVelScale:Vector="0,0,0"
    ExternalForce="0,0,0"
   ) -> (Pose) {
    angularDampingOverride=1
    angularSpringConstant=11000
    bAngularSpring=true
    bOverrideAngularDamping=true
    bUseGravityOverride=true
    bUseSphericalLimits=true
    gravityScale=0.35
    simulationSpace=RootRelative
    alphaBoolBlend {blendOption=Linear}
    boundBone {boneName=dyn_strap_leg_r_02}
    componentAppliedLinearAccClamp=<9500 9500 9500>
    componentLinearAccScale=<1 1 1>
    componentLinearVelScale=<1 1 1>
    gravityOverride {z=-980}
    physicsBodyDefinitions [{
      collisionType=CustomSphere
      sphereCollisionRadius=1.1
      boundBone {boneName=dyn_strap_leg_r_02}
      boxExtents=<15 15 15>
      constraintSetup {angularLimitsMax=<2 2 25>, angularLimitsMin=<-2 -20 -25>, angularTarget=<1 0 0>}
      localJointOffset=<4 0 0>
     }]
    sphericalLimits [
     {limitRadius=8, drivingBone {boneName=dyn_frather_leg_r}, sphereLocalOffset=<1 0 -5>}
     {limitRadius=5, drivingBone {boneName=calf_r}, sphereLocalOffset=<5 1 0>}
    ]
   }
   n11 = Anim Dynamics - Bone: dyn_frather_leg_r!(
    SimulationSpace:AnimPhysSimSpaceType Enum
    ComponentPose=@n10
    Alpha=1
    GravityOverride:Vector="0,0,0"
    ComponentLinearVelScale:Vector="0,0,0"
    ExternalForce="0,0,0"
   ) -> (Pose) {
    angularDampingOverride=1
    angularSpringConstant=11000
    bAngularSpring=true
    bOverrideAngularDamping=true
    bUseGravityOverride=true
    gravityScale=0.35
    simulationSpace=RootRelative
    alphaBoolBlend {blendOption=Linear}
    boundBone {boneName=dyn_frather_leg_r}
    componentAppliedLinearAccClamp=<9500 9500 9500>
    componentLinearAccScale=<1 1 1>
    componentLinearVelScale=<1 1 1>
    gravityOverride {z=-980}
    physicsBodyDefinitions [{
      boundBone {boneName=dyn_frather_leg_r}
      boxExtents=<30 30 30>
      constraintSetup {angularLimitsMax=<3 10 10>, angularLimitsMin=<-3 -10 -10>, angularTarget=<1 0 0>}
      localJointOffset=<2 0 0>
     }]
   }
   n12 = Anim Dynamics - Bone: dyn_frather_leg_l!(
    SimulationSpace:AnimPhysSimSpaceType Enum
    ComponentPose=@n11
    Alpha=1
    GravityOverride:Vector="0,0,0"
    ComponentLinearVelScale:Vector="0,0,0"
    ExternalForce="0,0,0"
   ) -> (Pose) {
    angularDampingOverride=1
    angularSpringConstant=11000
    bAngularSpring=true
    bOverrideAngularDamping=true
    bUseGravityOverride=true
    gravityScale=0.35
    simulationSpace=RootRelative
    alphaBoolBlend {blendOption=Linear}
    boundBone {boneName=dyn_frather_leg_l}
    componentAppliedLinearAccClamp=<9500 9500 9500>
    componentLinearAccScale=<1 1 1>
    componentLinearVelScale=<1 1 1>
    gravityOverride {z=-980}
    physicsBodyDefinitions [{
      boundBone {boneName=dyn_frather_leg_l}
      boxExtents=<30 30 30>
      constraintSetup {angularLimitsMax=<3 10 10>, angularLimitsMin=<-3 -10 -10>, angularTarget=<1 0 0>}
      localJointOffset=<2 0 0>
     }]
   }
   n13 = Anim Dynamics - Bone: dyn_strap_leg_l_01!(
    SimulationSpace:AnimPhysSimSpaceType Enum
    ComponentPose=@n12
    Alpha=0.9
    GravityOverride:Vector="0,0,0"
    ComponentLinearVelScale:Vector="0,0,0"
    ExternalForce="0,0,0"
   ) -> (Pose) {
    angularDampingOverride=1
    angularSpringConstant=11000
    bAngularSpring=true
    bOverrideAngularDamping=true
    bUseGravityOverride=true
    bUseSphericalLimits=true
    gravityScale=0.35
    simulationSpace=RootRelative
    alphaBoolBlend {blendOption=Linear}
    boundBone {boneName=dyn_strap_leg_l_01}
    componentAppliedLinearAccClamp=<9500 9500 9500>
    componentLinearAccScale=<1 1 1>
    componentLinearVelScale=<1 1 1>
    gravityOverride {z=-980}
    physicsBodyDefinitions [{
      collisionType=CustomSphere
      sphereCollisionRadius=1
      boundBone {boneName=dyn_strap_leg_l_01}
      boxExtents=<20 30 30>
      constraintSetup {angularLimitsMax=<0 0 15>, angularLimitsMin=<0 -10 -15>, angularTarget=<1 0 0>}
      localJointOffset=<4 0 0>
     }]
    sphericalLimits [{limitRadius=5, drivingBone {boneName=calf_l}, sphereLocalOffset=<-5 -1 0>}]
   }
   n14 = Anim Dynamics - Bone: dyn_strap_leg_l_02!(
    SimulationSpace:AnimPhysSimSpaceType Enum
    ComponentPose=@n13
    Alpha=0.9
    GravityOverride:Vector="0,0,0"
    ComponentLinearVelScale:Vector="0,0,0"
    ExternalForce="0,0,0"
   ) -> (Pose) {
    angularDampingOverride=1
    angularSpringConstant=11000
    bAngularSpring=true
    bOverrideAngularDamping=true
    bUseGravityOverride=true
    bUseSphericalLimits=true
    gravityScale=0.35
    simulationSpace=RootRelative
    alphaBoolBlend {blendOption=Linear}
    boundBone {boneName=dyn_strap_leg_l_02}
    componentAppliedLinearAccClamp=<9500 9500 9500>
    componentLinearAccScale=<1 1 1>
    componentLinearVelScale=<1 1 1>
    gravityOverride {z=-980}
    physicsBodyDefinitions [{
      collisionType=CustomSphere
      sphereCollisionRadius=1.1
      boundBone {boneName=dyn_strap_leg_l_02}
      boxExtents=<20 30 30>
      constraintSetup {angularLimitsMax=<0 0 15>, angularLimitsMin=<0 -10 -15>, angularTarget=<1 0 0>}
      localJointOffset=<4 0 0>
     }]
    sphericalLimits [
     {limitRadius=8.5, drivingBone {boneName=dyn_frather_leg_l}, sphereLocalOffset=<-1 0 -5>}
     {limitRadius=5.5, drivingBone {boneName=calf_l}, sphereLocalOffset=<-5 -1 0>}
    ]
   }
   n15 = Anim Dynamics - Bone: dyn_strap_leg_l_03!(
    SimulationSpace:AnimPhysSimSpaceType Enum
    ComponentPose=@n14
    Alpha=0.9
    GravityOverride:Vector="0,0,0"
    ComponentLinearVelScale:Vector="0,0,0"
    ExternalForce="0,0,0"
   ) -> (Pose) {
    angularDampingOverride=1
    angularSpringConstant=11000
    bAngularSpring=true
    bOverrideAngularDamping=true
    bUseGravityOverride=true
    bUseSphericalLimits=true
    gravityScale=0.35
    simulationSpace=RootRelative
    alphaBoolBlend {blendOption=Linear}
    boundBone {boneName=dyn_strap_leg_l_03}
    componentAppliedLinearAccClamp=<9500 9500 9500>
    componentLinearAccScale=<1 1 1>
    componentLinearVelScale=<1 1 1>
    gravityOverride {z=-980}
    physicsBodyDefinitions [{
      sphereCollisionRadius=1.1
      boundBone {boneName=dyn_strap_leg_l_03}
      boxExtents=<15 15 15>
      constraintSetup {angularLimitsMax=<2 0 25>, angularLimitsMin=<-2 -20 -25>, angularTarget=<1 0 0>}
      localJointOffset=<4 0 0>
     }]
    sphericalLimits [
     {limitRadius=9, drivingBone {boneName=dyn_frather_leg_l}, sphereLocalOffset=<-1 0 -5>}
     {limitRadius=6, drivingBone {boneName=calf_l}, sphereLocalOffset=<-5 -1 0>}
    ]
   }
   n16 = Anim Dynamics - Bone: dyn_belt_l!(
    SimulationSpace:AnimPhysSimSpaceType Enum
    ComponentPose=@n15
    Alpha=1
    GravityOverride:Vector="0,0,0"
    ComponentLinearVelScale:Vector="0,0,0"
    ExternalForce="0,0,0"
   ) -> (Pose) {
    angularDampingOverride=1
    angularSpringConstant=11000
    bAngularSpring=true
    bOverrideAngularDamping=true
    bUseGravityOverride=true
    gravityScale=0.35
    simulationSpace=RootRelative
    alphaBoolBlend {blendOption=Linear}
    boundBone {boneName=dyn_belt_l}
    componentAppliedLinearAccClamp=<9500 9500 9500>
    componentLinearAccScale=<1 1 1>
    componentLinearVelScale=<1 1 1>
    gravityOverride {z=-980}
    physicsBodyDefinitions [{
      boundBone {boneName=dyn_belt_l}
      boxExtents=<40 40 40>
      constraintSetup {angularLimitsMax=<5 0 5>, angularLimitsMin=<-5 -20 -5>, angularTarget=<1 0 0>}
      localJointOffset=<3 0 0>
     }]
   }
   n17 = Anim Dynamics - Bone: dyn_frather_shoulder_l!(
    SimulationSpace:AnimPhysSimSpaceType Enum
    ComponentPose=@n16
    Alpha=0.9
    GravityOverride:Vector="0,0,0"
    ComponentLinearVelScale:Vector="0,0,0"
    ExternalForce="0,0,0"
   ) -> (Pose) {
    angularDampingOverride=1
    angularSpringConstant=11000
    bAngularSpring=true
    bOverrideAngularDamping=true
    bUseGravityOverride=true
    gravityScale=0.35
    simulationSpace=RootRelative
    alphaBoolBlend {blendOption=Linear}
    boundBone {boneName=dyn_frather_shoulder_l}
    componentAppliedLinearAccClamp=<9500 9500 9500>
    componentLinearAccScale=<1 1 1>
    componentLinearVelScale=<1 1 1>
    gravityOverride {z=-980}
    physicsBodyDefinitions [{
      boundBone {boneName=dyn_frather_shoulder_l}
      boxExtents=<40 40 40>
      constraintSetup {angularLimitsMax=<5 5 15>, angularLimitsMin=<-5 -15 -15>, angularTarget=<1 0 0>}
      localJointOffset=<2 0 0>
     }]
   }
   n18 = Anim Dynamics - Bone: dyn_frather_elbow_l!(
    SimulationSpace:AnimPhysSimSpaceType Enum
    ComponentPose=@n17
    Alpha=1
    GravityOverride:Vector="0,0,0"
    ComponentLinearVelScale:Vector="0,0,0"
    ExternalForce="0,0,0"
   ) -> (Pose) {
    angularDampingOverride=1
    angularSpringConstant=11000
    bAngularSpring=true
    bOverrideAngularDamping=true
    bUseGravityOverride=true
    gravityScale=0.35
    simulationSpace=RootRelative
    alphaBoolBlend {blendOption=Linear}
    boundBone {boneName=dyn_frather_elbow_l}
    componentAppliedLinearAccClamp=<9500 9500 9500>
    componentLinearAccScale=<1 1 1>
    componentLinearVelScale=<1 1 1>
    gravityOverride {z=-980}
    physicsBodyDefinitions [{
      boundBone {boneName=dyn_frather_elbow_l}
      boxExtents=<30 30 30>
      constraintSetup {angularLimitsMax=<3 10 10>, angularLimitsMin=<-3 -10 -10>, angularTarget=<1 0 0>}
      localJointOffset=<2 0 0>
     }]
   }
   n19 = Anim Dynamics - Bone: dyn_frather_wrist_l!(
    SimulationSpace:AnimPhysSimSpaceType Enum
    ComponentPose=@n18
    Alpha=1
    GravityOverride:Vector="0,0,0"
    ComponentLinearVelScale:Vector="0,0,0"
    ExternalForce="0,0,0"
   ) -> (Pose) {
    angularDampingOverride=1
    angularSpringConstant=11000
    bAngularSpring=true
    bOverrideAngularDamping=true
    bUseGravityOverride=true
    gravityScale=0.35
    simulationSpace=RootRelative
    alphaBoolBlend {blendOption=Linear}
    boundBone {boneName=dyn_frather_wrist_l}
    componentAppliedLinearAccClamp=<9500 9500 9500>
    componentLinearAccScale=<1 1 1>
    componentLinearVelScale=<1 1 1>
    gravityOverride {z=-980}
    physicsBodyDefinitions [{
      boundBone {boneName=dyn_frather_wrist_l}
      boxExtents=<30 30 30>
      constraintSetup {angularLimitsMax=<0 3 15>, angularLimitsMin=<0 -5 -15>, angularTarget=<1 0 0>}
      localJointOffset=<2 0 0>
     }]
   }
   n20 = Get GravityOverride()
   n21 = Get Sim Space Settings() -> (Sim Space Settings:Sim Space Settings Structure)
   n22 = RigidBody!(ComponentPose=@n19, Alpha, OverrideWorldGravity:Vector=@n20, ExternalForce, SimSpaceSettings:Sim Space Settings Structure=@n21) -> (Pose) {
    bClampLinearTranslationLimitToRefPose=true
    bOverrideWorldGravity=true
    overridePhysicsAsset=&/F_MED_Nevermore_Physics
    simulationSpace=BaseBoneSpace
    alphaBoolBlend {blendOption=Linear}
    baseBoneRef {boneName=pelvis}
    simSpaceSettings {maxAngularAcceleration=1, maxAngularVelocity=3, maxLinearAcceleration=1000, maxLinearVelocity=500, worldAlpha=1, externalLinearDragV=<1 1 1>}
   }
   n23 = RigidBody(ComponentPose=@n22, OverrideWorldGravity=@n20, SimSpaceSettings=@n21) {
    bClampLinearTranslationLimitToRefPose=true
    bOverrideWorldGravity=true
    overridePhysicsAsset=&/F_MED_Nevermore_Neck_Physics
    simulationSpace=BaseBoneSpace
    alphaBoolBlend {blendOption=Linear}
    baseBoneRef {boneName=spine_05}
    simSpaceSettings {maxAngularAcceleration=1, maxAngularVelocity=3, maxLinearAcceleration=1000, maxLinearVelocity=500, worldAlpha=1, externalLinearDragV=<1 1 1>}
   }
   n24 = Anim Dynamics - Bone: dyn_strap_neck_r_01!(
    SimulationSpace:AnimPhysSimSpaceType Enum
    ComponentPose=@n23
    bAlphaBoolEnabled:bool=True
    GravityOverride:Vector="0,0,0"
    ComponentLinearVelScale:Vector="0,0,0"
    ExternalForce="0,0,0"
   ) -> (Pose) {
    alphaInputType=Bool
    angularDampingOverride=1
    angularSpringConstant=11000
    bAngularSpring=true
    bOverrideAngularDamping=true
    bUseGravityOverride=true
    gravityScale=0.35
    simulationSpace=RootRelative
    alphaBoolBlend {blendInTime=0.5, blendOption=QuadraticInOut, blendOutTime=1}
    boundBone {boneName=dyn_strap_neck_r_01}
    componentAppliedLinearAccClamp=<9500 9500 9500>
    componentLinearAccScale=<1 1 1>
    componentLinearVelScale=<1 1 1>
    gravityOverride {z=-980}
    physicsBodyDefinitions [{
      sphereCollisionRadius=1
      boundBone {boneName=dyn_strap_neck_r_01}
      boxExtents=<20 30 30>
      constraintSetup {angularLimitsMax=<0 4 0>, angularLimitsMin=<0 -4 -8>, angularTarget=<1 0 0>}
      localJointOffset=<4 0 0>
     }]
   }
   n25 = Anim Dynamics - Bone: dyn_strap_neck_r_02!(
    SimulationSpace:AnimPhysSimSpaceType Enum
    ComponentPose=@n24
    bAlphaBoolEnabled:bool=True
    GravityOverride:Vector="0,0,0"
    ComponentLinearVelScale:Vector="0,0,0"
    ExternalForce="0,0,0"
   ) -> (Pose) {
    alphaInputType=Bool
    angularDampingOverride=1
    angularSpringConstant=11000
    bAngularSpring=true
    bOverrideAngularDamping=true
    bUseGravityOverride=true
    gravityScale=0.35
    simulationSpace=RootRelative
    alphaBoolBlend {blendInTime=0.5, blendOption=QuadraticInOut, blendOutTime=1}
    boundBone {boneName=dyn_strap_neck_r_02}
    componentAppliedLinearAccClamp=<9500 9500 9500>
    componentLinearAccScale=<1 1 1>
    componentLinearVelScale=<1 1 1>
    gravityOverride {z=-980}
    physicsBodyDefinitions [{
      sphereCollisionRadius=1.1
      boundBone {boneName=dyn_strap_neck_r_02}
      boxExtents=<15 15 15>
      constraintSetup {angularLimitsMax=<1 8 1>, angularLimitsMin=<-1 -8 -15>, angularTarget=<1 0 0>}
      localJointOffset=<4 0 0>
     }]
   }
   n26 = Anim Dynamics - Bone: dyn_strap_neck_l_01!(
    SimulationSpace:AnimPhysSimSpaceType Enum
    ComponentPose=@n25
    bAlphaBoolEnabled:bool=True
    GravityOverride:Vector="0,0,0"
    ComponentLinearVelScale:Vector="0,0,0"
    ExternalForce="0,0,0"
   ) -> (Pose) {
    alphaInputType=Bool
    angularDampingOverride=1
    angularSpringConstant=11000
    bAngularSpring=true
    bOverrideAngularDamping=true
    bUseGravityOverride=true
    gravityScale=0.35
    simulationSpace=RootRelative
    alphaBoolBlend {blendInTime=0.5, blendOption=QuadraticInOut, blendOutTime=1}
    boundBone {boneName=dyn_strap_neck_l_01}
    componentAppliedLinearAccClamp=<9500 9500 9500>
    componentLinearAccScale=<1 1 1>
    componentLinearVelScale=<1 1 1>
    gravityOverride {z=-980}
    physicsBodyDefinitions [{
      sphereCollisionRadius=1
      boundBone {boneName=dyn_strap_neck_l_01}
      boxExtents=<20 30 30>
      constraintSetup {angularLimitsMax=<0 4 8>, angularLimitsMin=<0 -4 0>, angularTarget=<1 0 0>}
      localJointOffset=<4 0 0>
     }]
   }
   n27 = Anim Dynamics - Bone: dyn_strap_neck_l_02!(
    SimulationSpace:AnimPhysSimSpaceType Enum
    ComponentPose=@n26
    bAlphaBoolEnabled:bool=True
    GravityOverride:Vector="0,0,0"
    ComponentLinearVelScale:Vector="0,0,0"
    ExternalForce="0,0,0"
   ) -> (Pose) {
    alphaInputType=Bool
    angularDampingOverride=1
    angularSpringConstant=11000
    bAngularSpring=true
    bOverrideAngularDamping=true
    bUseGravityOverride=true
    gravityScale=0.35
    simulationSpace=RootRelative
    alphaBoolBlend {blendInTime=0.5, blendOption=QuadraticInOut, blendOutTime=1}
    boundBone {boneName=dyn_strap_neck_l_02}
    componentAppliedLinearAccClamp=<9500 9500 9500>
    componentLinearAccScale=<1 1 1>
    componentLinearVelScale=<1 1 1>
    gravityOverride {z=-980}
    physicsBodyDefinitions [{
      sphereCollisionRadius=1.1
      boundBone {boneName=dyn_strap_neck_l_02}
      boxExtents=<20 30 30>
      constraintSetup {angularLimitsMax=<1 8 15>, angularLimitsMin=<-1.5 -8 -1>, angularTarget=<1 0 0>}
      localJointOffset=<4 0 0>
     }]
   }
   n28 = Anim Dynamics - Bone: dyn_strap_neck_l_03!(
    SimulationSpace:AnimPhysSimSpaceType Enum
    ComponentPose=@n27
    bAlphaBoolEnabled:bool=True
    GravityOverride:Vector="0,0,0"
    ComponentLinearVelScale:Vector="0,0,0"
    ExternalForce="0,0,0"
   ) -> (Pose) {
    alphaInputType=Bool
    angularDampingOverride=1
    angularSpringConstant=11000
    bAngularSpring=true
    bOverrideAngularDamping=true
    bUseGravityOverride=true
    gravityScale=0.35
    simulationSpace=RootRelative
    alphaBoolBlend {blendInTime=0.5, blendOption=QuadraticInOut, blendOutTime=1}
    boundBone {boneName=dyn_strap_neck_l_03}
    componentAppliedLinearAccClamp=<9500 9500 9500>
    componentLinearAccScale=<1 1 1>
    componentLinearVelScale=<1 1 1>
    gravityOverride {z=-980}
    physicsBodyDefinitions [{
      sphereCollisionRadius=1.1
      boundBone {boneName=dyn_strap_neck_l_03}
      boxExtents=<15 15 15>
      constraintSetup {angularLimitsMax=<1.5 15 20>, angularLimitsMin=<-1.5 -15 -4>, angularTarget=<1 0 0>}
      localJointOffset=<4 0 0>
     }]
   }
   n29 = "Transform (Modify) Bone - Bone: surfaceconst_belt_l_front"!(ComponentPose=@n28, bAlphaBoolEnabled:bool=True, Translation="0,0,0", Rotation="5,10,0", Scale="1,1,1") -> (Pose) {
    alphaInputType=Bool
    rotationMode=BMM_Additive
    rotationSpace=BCS_BoneSpace
    scaleMode=BMM_Additive
    scaleSpace=BCS_BoneSpace
    translationMode=BMM_Additive
    translationSpace=BCS_BoneSpace
    alphaBoolBlend {blendInTime=0.5, blendOption=QuadraticInOut, blendOutTime=1}
    boneToModify {boneName=surfaceconst_belt_l_front}
    rotation {pitch=5, yaw=10}
   }
   n30 = "Transform (Modify) Bone - Bone: surfaceconst_belt_r_back"!(ComponentPose=@n29, bAlphaBoolEnabled:bool=True, Translation="0,1,0", Rotation="0,10,-20", Scale="1,1,1") -> (Pose) {
    alphaInputType=Bool
    rotationMode=BMM_Additive
    rotationSpace=BCS_BoneSpace
    scaleMode=BMM_Additive
    scaleSpace=BCS_BoneSpace
    translationMode=BMM_Additive
    translationSpace=BCS_BoneSpace
    alphaBoolBlend {blendInTime=0.5, blendOption=QuadraticInOut, blendOutTime=1}
    boneToModify {boneName=surfaceconst_belt_r_back}
    rotation {roll=-20, yaw=10}
    translation {y=1}
   }
   Output Pose(Result=Component To Local(ComponentPose=@n30)) {layerGroup=None}
   n31 = Layered blend per bone!(
    BasePose=Copy Pose From Mesh() {bCopyCurves=true, bUseAttachedParent=true}
    BlendPoses_0=Copy Pose From Mesh(SourceMeshComponent=Get FaceAcc_Mesh() -> (FaceAcc_Mesh:Skeletal Mesh Component Object Reference))
    BlendPoses_1:Pose Link Structure=Copy Pose From Mesh(SourceMeshComponent=Get Head_Mesh() -> (Head_Mesh:Skeletal Mesh Component Object Reference))
    BlendWeights_0
    BlendWeights_1:float=1
   ) -> (Pose) {
    blendPoses [{linkId=-1, sourceLinkId=-1}, {linkId=-1, sourceLinkId=-1}]
    blendWeights [1,1]
    layerSetup [
     {branchFilters [
       {boneName=base_hair_front}
       {boneName=dyn_hair_front_01}
       {boneName=dyn_hair_front_02}
       {boneName=dyn_hair_front_03}
       {boneName=base_hair_mid}
       {boneName=dyn_hair_mid_01}
       {boneName=dyn_hair_mid_02}
      ]}
     {branchFilters [
       {boneName=FACIAL_C_FacialRoot}
       {boneName=FACIAL_C_UnderChin}
       {boneName=FACIAL_L_UnderChin}
       {boneName=FACIAL_R_UnderChin}
       {boneName=FACIAL_C_Jaw}
       {boneName=FACIAL_C_Jawline}
       {boneName=FACIAL_L_ChinSide}
       {boneName=FACIAL_R_ChinSide}
       {boneName=FACIAL_C_Chin}
       {boneName=FACIAL_L_Jawline}
       {boneName=FACIAL_R_Jawline}
      ]}
    ]
   }
   Save cached pose 'StartPose'(Pose=@n31) {cacheName=StartPose}
  ]}
 EventGraph {nodes [
   Event OnCharacterCustomizationLoaded()
   > n1 = Sequence()
   > n2 = Cast To PlayerPawn_Athena(Object=TryGetPawnOwner())
   > Set Head_Mesh(Head_Mesh:Skeletal Mesh Component Object Reference=GetSkeletalMeshForPartType(self=@n2, PartType=Head)) -> (Output_Get:Skeletal Mesh Component Object Reference)
   n3 = Cast To PlayerPawn_Athena(execute=@n1.then_1, Object=TryGetPawnOwner())
   > Set FaceAcc_Mesh(FaceAcc_Mesh:Skeletal Mesh Component Object Reference=GetSkeletalMeshForPartType(self=@n3, PartType=Face)) -> (Output_Get:Skeletal Mesh Component Object Reference)
   Event BlueprintUpdateAnimation()
   > n4 = Sequence!(execute) -> (then_0, then_1, then_2:Exec, then_3:Exec)
   > n5 = "Gravity Override State Function (Parameter Asset) (Nativized)"(
    CharacterPartAnimInstance=Self-Reference()
    GameTimeInSeconds=GetGameTimeInSeconds()
    Parameters=&/F_MED_Nevermore_Parameters
   )
   > Set GravityOverride(GravityOverride=@n5)
   n6 = Event BlueprintProcessAnimInputEvent()
   n7 = IsNamed(self=@n6.AnimInputEvent, InEventName=aie.Player.Zipline.Entered)
   n8 = OR Boolean(A=Get bIsMoving2D(), B=Get bIsTacticalSprinting())
   n9 = Get bIsJumping()
   n10 = Branch(execute=@n4.then_1, Condition=OR Boolean(A=@n7, B=AND Boolean!(A=@n7, B=@n9, C:bool=@n8) -> (ReturnValue)))
   > Set Componen Linear Vel Scale(Componen Linear Vel Scale:Vector="0, 0, 0") -> (Output_Get:Vector)
   Branch(execute=@n10.else, Condition=IsNamed(self=@n6.AnimInputEvent, InEventName=aie.Player.Zipline.Exited))
   > Set Componen Linear Vel Scale(Componen Linear Vel Scale="1,1,1")
   n11 = Branch(execute=@n4.then_2, Condition=Get bIsInFrontEnd())
   > Set Simulation Space(Simulation Space:Byte=2) -> (Output_Get:Byte=0)
   > Set FE_Alpha(FE_Alpha:double=0.75) -> (Output_Get:double)
   Set Simulation Space(execute=@n11.else, Simulation Space=3) -> (Output_Get=0)
   > Set FE_Alpha(FE_Alpha=1)
   Set LOD1Plus(execute=@n4.then_3, LOD1Plus:bool="integer >= integer"(A=Get CurrentLOD(), B=1)) -> (Output_Get:bool)
  ]}
]
variables [
 Componen Linear Vel Scale {type=Vector}
 FaceAcc_Mesh {type=SkeletalMeshComponent}
 FE_Alpha {type=number}
 Gravity Override Params {type=GravityOverrideParamsStruct}
 GravityOverride {type=Vector}
 Head_Mesh {type=SkeletalMeshComponent}
 LOD1Plus {type=boolean}
 RBAN Params {
  type=RBANParamsStruct
  default {
   dBNO {simSpaceSettings {
     maxAngularAcceleration=1000
     maxAngularVelocity=3
     maxLinearAcceleration=300
     maxLinearVelocity=700
     velocityScaleZ=0
     externalLinearDragV=<0 0.5 0>
    }}
   dBNO_Carried {simSpaceSettings {
     maxAngularAcceleration=1000
     maxAngularVelocity=3
     maxLinearAcceleration=300
     maxLinearVelocity=700
     velocityScaleZ=0
     externalLinearDragV=<0 0.5 0>
    }}
   emoteOrMelee {simSpaceSettings {
     maxAngularAcceleration=1000
     maxAngularVelocity=3
     maxLinearAcceleration=300
     maxLinearVelocity=700
     velocityScaleZ=0
     externalLinearDragV=<0 0.5 0>
    }}
   front_End {simSpaceSettings {
     maxAngularAcceleration=1000
     maxAngularVelocity=3
     maxLinearAcceleration=300
     maxLinearVelocity=700
     velocityScaleZ=0
     externalLinearDragV=<0 0.5 0>
    }}
   onGround_Crouch_Moving {simSpaceSettings {
     maxAngularAcceleration=1000
     maxAngularVelocity=3
     maxLinearAcceleration=300
     maxLinearVelocity=700
     velocityScaleZ=0
     externalLinearDragV=<0 0.5 0>
    }}
   onGround_Falling {simSpaceSettings {
     maxAngularAcceleration=1000
     maxAngularVelocity=3
     maxLinearAcceleration=300
     maxLinearVelocity=700
     velocityScaleZ=0
     externalLinearDragV=<0 0.5 0>
    }}
   onGround_Moving {simSpaceSettings {
     maxAngularAcceleration=1000
     maxAngularVelocity=3
     maxLinearAcceleration=300
     maxLinearVelocity=700
     velocityScaleZ=0
     externalLinearDragV=<0 0.5 0>
    }}
   onGround_Standing {simSpaceSettings {
     maxAngularAcceleration=1000
     maxAngularVelocity=3
     maxLinearAcceleration=300
     maxLinearVelocity=700
     velocityScaleZ=0
     externalLinearDragV=<0 0.5 0>
    }}
   rail_Grinding {simSpaceSettings {
     maxAngularAcceleration=1000
     maxAngularVelocity=3
     maxLinearAcceleration=300
     maxLinearVelocity=700
     velocityScaleZ=0
     worldAlpha=0.1
     externalLinearDragV=<0 0.5 0>
    }}
   riding_Moving {simSpaceSettings {
     maxAngularAcceleration=1000
     maxAngularVelocity=3
     maxLinearAcceleration=300
     maxLinearVelocity=700
     velocityScaleZ=0
     externalLinearDragV=<0 0.5 0>
    }}
   riding_Moving_Targeting {simSpaceSettings {
     maxAngularAcceleration=1000
     maxAngularVelocity=3
     maxLinearAcceleration=300
     maxLinearVelocity=700
     velocityScaleZ=0
     externalLinearDragV=<0 0.5 0>
    }}
   riding_Standing {simSpaceSettings {
     maxAngularAcceleration=1000
     maxAngularVelocity=3
     maxLinearAcceleration=300
     maxLinearVelocity=700
     velocityScaleZ=0
     externalLinearDragV=<0 0.5 0>
    }}
   skydive_DivingDown {simSpaceSettings {
     maxAngularAcceleration=1000
     maxAngularVelocity=3
     maxLinearAcceleration=300
     maxLinearVelocity=700
     velocityScaleZ=0
     externalLinearDragV=<0 0.5 0>
    }}
   skydive_FloatingDown {
    componentLinearAccScale=<0 0.5 0>
    componentLinearVelScale=<0 0.5 0>
    simSpaceSettings {
     maxAngularAcceleration=1000
     maxAngularVelocity=3
     maxLinearAcceleration=300
     maxLinearVelocity=700
     velocityScaleZ=0
     externalLinearDragV=<0 0.5 0>
    }
   }
   skydive_GoingUp {simSpaceSettings {
     maxAngularAcceleration=1000
     maxAngularVelocity=3
     maxLinearAcceleration=300
     maxLinearVelocity=700
     velocityScaleZ=0
     externalLinearDragV=<0 0.5 0>
    }}
   skydive_Parachuting {
    componentLinearAccScale=<0 0.1 0.3>
    componentLinearVelScale=<0 0.1 0.3>
    simSpaceSettings {
     maxAngularAcceleration=1000
     maxAngularVelocity=3
     maxLinearAcceleration=300
     maxLinearVelocity=700
     velocityScaleZ=0
     externalLinearDragV=<0 0.5 0>
    }
   }
   skydive_Surfing {simSpaceSettings {
     maxAngularAcceleration=1000
     maxAngularVelocity=3
     maxLinearAcceleration=300
     maxLinearVelocity=700
     velocityScaleZ=0
     externalLinearDragV=<0 0.5 0>
    }}
   swimming {simSpaceSettings {
     maxAngularAcceleration=1000
     maxAngularVelocity=3
     maxLinearAcceleration=300
     maxLinearVelocity=700
     velocityScaleZ=0
     externalLinearDragV=<0 0.5 0>
    }}
   tactical_Sprint {simSpaceSettings {
     maxAngularAcceleration=1000
     maxAngularVelocity=3
     maxLinearAcceleration=300
     maxLinearVelocity=700
     velocityScaleZ=0
     externalLinearDragV=<0 0.5 0>
    }}
   ziplining {simSpaceSettings {
     maxAngularAcceleration=1000
     maxAngularVelocity=3
     maxLinearAcceleration=300
     maxLinearVelocity=700
     velocityScaleZ=0
     externalLinearDragV=<0 0.5 0>
    }}
  }
 }
 Sim Space Settings {
  type=SimSpaceSettings
  default {
   dampingAlpha=1
   maxAngularAcceleration=10000
   maxAngularVelocity=10000
   maxLinearAcceleration=10000
   maxLinearVelocity=10000
   velocityScaleZ=1
   worldAlpha=0
   externalLinearDragV=<0 0 0>
  }
 }
 Simulation Space {type=integer}
]
