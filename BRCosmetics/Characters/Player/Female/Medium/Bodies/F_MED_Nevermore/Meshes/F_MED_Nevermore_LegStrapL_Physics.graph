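# Physics asset for the left leg strap on the female medium Nevermore character body, set up as a short jointed capsule chain.
# Three dyn_strap capsules hang from a tiny base capsule, with narrow swing and twist limits that loosen slightly in one swing axis toward the tip.
# The base joins the upper thigh twist bone with free angular motion.
# Larger thigh and calf capsules approximate the leg, the calf having no constraint.
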PhysicsAsset /BRCosmetics/Characters/Player/Female/Medium/Bodies/F_MED_Nevermore/Meshes/F_MED_Nevermore_LegStrapL_Physics
bodies {
 base_strap_leg_l [{shapeType=Capsule, length=0.505, radius=0.505, shapeName=base_strap_leg_l_capsule, rotation=<rot 0 -90 0>}]
 calf_l [{shapeType=Capsule, length=33, radius=7, shapeName=calf_l_capsule, center=<-22.2437 -2.4739 -4.0053>, rotation=<rot 70 180 180>}]
 dyn_strap_leg_l_01 [{shapeType=Capsule, length=3, radius=1.5, shapeName=dyn_strap_leg_l_01_capsule, center=<2.9616 0 0.8>, rotation=<rot 90 0 0>}]
 dyn_strap_leg_l_02 [{shapeType=Capsule, length=4, radius=1.8, shapeName=dyn_strap_leg_l_02_capsule, center=<3.964 0 0.8>, rotation=<rot 90 0 0>}]
 dyn_strap_leg_l_03 [{shapeType=Capsule, length=6, radius=1.8, shapeName=dyn_strap_leg_l_03_capsule, center=<5.6473 0 0.8>, rotation=<rot 90 0 0>}]
 thigh_twist_01_l [{shapeType=Capsule, length=22, radius=6.5, shapeName=thigh_twist_01_l_capsule, center=<-7.8809 0 1.1562>, rotation=<rot 90 0 0>}]
}
constraints [
 {bone1Name=dyn_strap_leg_l_03, bone2Name=dyn_strap_leg_l_02, swing1LimitDegrees=20, swing2LimitDegrees=10, twistLimitDegrees=5}
 {bone1Name=dyn_strap_leg_l_02, bone2Name=dyn_strap_leg_l_01, swing1LimitDegrees=15, swing2LimitDegrees=10, twistLimitDegrees=5}
 {bone1Name=dyn_strap_leg_l_01, bone2Name=base_strap_leg_l, swing1LimitDegrees=10, swing2LimitDegrees=10, twistLimitDegrees=5}
 {
  bone1Name=base_strap_leg_l
  bone2Name=thigh_twist_01_l
  swing1LimitDegrees=45
  swing1Motion=Free
  swing2LimitDegrees=45
  swing2Motion=Free
  twistLimitDegrees=45
  twistMotion=Free
 }
]
