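# Physics asset for the right leg strap on the F_MED_Nevermore female player body cosmetic, setting up physical motion for the strap.
# Capsule bodies cover two dynamic strap bones, a tiny strap base, and the right thigh twist and calf, which form the leg volumes.
# The strap bones chain to the base with tight swing and twist limits, while the base joins the thigh with free angular motion.
# The calf body has no constraint.
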PhysicsAsset /BRCosmetics/Characters/Player/Female/Medium/Bodies/F_MED_Nevermore/Meshes/F_MED_Nevermore_LegStrapR_Physics
bodies {
 base_strap_leg_r [{shapeType=Capsule, length=0.505, radius=0.505, shapeName=base_strap_leg_r_capsule, rotation=<rot 0 -90 0>}]
 calf_r [{shapeType=Capsule, length=33, radius=7, shapeName=calf_r_capsule, center=<22.2437 2.4739 4.0053>, rotation=<rot 110 0 0>}]
 dyn_strap_leg_r_01 [{shapeType=Capsule, length=2, radius=1.8, shapeName=dyn_strap_leg_r_01_capsule, center=<3.362 0 0.7458>, rotation=<rot 90 0 0>}]
 dyn_strap_leg_r_02 [{shapeType=Capsule, length=5, radius=1.8, shapeName=dyn_strap_leg_r_02_capsule, center=<4.5455 0 0.6066>, rotation=<rot 90 0 0>}]
 thigh_twist_01_r [{shapeType=Capsule, length=22, radius=6.5, shapeName=thigh_twist_01_r_capsule, center=<7.8554 0 -1.4477>, rotation=<rot 85 0 0>}]
}
constraints [
 {
  bone1Name=base_strap_leg_r
  bone2Name=thigh_twist_01_r
  swing1LimitDegrees=45
  swing1Motion=Free
  swing2LimitDegrees=45
  swing2Motion=Free
  twistLimitDegrees=45
  twistMotion=Free
 }
 {bone1Name=dyn_strap_leg_r_01, bone2Name=base_strap_leg_r, swing1LimitDegrees=15, swing2LimitDegrees=10, twistLimitDegrees=5}
 {bone1Name=dyn_strap_leg_r_02, bone2Name=dyn_strap_leg_r_01, swing1LimitDegrees=20, swing2LimitDegrees=10, twistLimitDegrees=5}
]
