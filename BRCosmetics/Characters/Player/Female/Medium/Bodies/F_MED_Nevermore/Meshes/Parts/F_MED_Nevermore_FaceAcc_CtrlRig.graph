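# Control Rig for the Nevermore female face accessory that adjusts hair bones as the head twists in one direction.
# A reusable DeltaRotation function swing-twist decomposes head rotation relative to head_null and remaps one Euler axis to a clamped weight.
# That weight drives hair aim-target offsets, aims the mid hair at the left neck strap, and rotates and shrinks the back hair.
# The back hair base always aims at its own target.
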ControlRigBlueprint /BRCosmetics/Characters/Player/Female/Medium/Bodies/F_MED_Nevermore/Meshes/Parts/F_MED_Nevermore_FaceAcc_CtrlRig
graphs [
 New Function_ContainedGraph {nodes [
   n1 = RigVMFunctionEntryNode()
   RigVMFunction_MathBoolToInteger = &RigVMFunction_MathBoolToInteger(Value=@n1.onlyTwist)
   RigVMFunction_MathQuaternionSwingTwist = &RigVMFunction_MathQuaternionSwingTwist(
    twistAxis=@n1.twistAxis
    Input=&RigUnit_GetRelativeTransformForItem(
     bChildInitial=False
     bParentInitial=False
     child=(Type=Bone,Name="None")
     parent=(Type=Bone,Name="None")
     child.Name=@n1.child
     parent.Name=@n1.parent
     parent.Type=@n1.parentType
    ).RelativeTransform.Rotation
   )
   RigVMFunction_MathQuaternionToEuler = &RigVMFunction_MathQuaternionToEuler(
    Value=RigVMDispatchNode(
     Values=((),())
     Index=@RigVMFunction_MathBoolToInteger.Result
     Values.0=@RigVMFunction_MathQuaternionSwingTwist.Swing
     Values.1=@RigVMFunction_MathQuaternionSwingTwist.Twist
    ) {template="DISPATCH_RigVMDispatch_SelectInt32(in Index,in Values,out Result)"}.Result
    RotationOrder=@n1.Rotation Order
   )
   > DISPATCH_RigVMDispatch_Print = RigVMDispatchNode(ScreenColor=(R=1,G=1,B=1,A=1), ScreenDuration=1.5, Enabled=@n1.debugRotation, Value=@RigVMFunction_MathQuaternionToEuler.Result) {template="DISPATCH_RigVMDispatch_Print(in Prefix,in Value,in Enabled,in ScreenDuration,in ScreenColor)"}
   Clamp = &RigVMFunction_MathIntClamp(Value=@n1.primAxis)
   DISPATCH_RigVMDispatch_SelectInt32 = RigVMDispatchNode(
    Values=(0,0,0)
    Index=@Clamp.Result
    Values.0=@RigVMFunction_MathQuaternionToEuler.Result.X
    Values.1=@RigVMFunction_MathQuaternionToEuler.Result.Y
    Values.2=@RigVMFunction_MathQuaternionToEuler.Result.Z
   ) {template="DISPATCH_RigVMDispatch_SelectInt32(in Index,in Values,out Result)"}
   Remap = &RigVMFunction_MathDoubleRemap(
    TargetMaximum=1
    TargetMinimum=0
    bClamp=False
    Value=@DISPATCH_RigVMDispatch_SelectInt32.Result
    SourceMaximum=@n1.max
    SourceMinimum=@n1.min
   )
   Multiply = &RigVMFunction_MathDoubleMul(A=&RigVMFunction_MathDoubleClamp(Value=@Remap.Result).Result, B=@n1.premult)
   > RigVMFunctionReturnNode(outputWeight=@Multiply.Result)
  ]}
 RigVMFunctionLibrary {nodes [DeltaRotation_2 = RigVMCollapseNode(ExecuteContext=(), Rotation Order=ZYX, premult=1, twistAxis=(X=1,Y=0,Z=0))]}
 RigVMModel {nodes [
   RigUnit_BeginExecution = &RigUnit_BeginExecution()
   RigVMFunction_Sequence = &RigVMFunction_Sequence(ExecuteContext=@RigUnit_BeginExecution.ExecutePin)
   DeltaRotation_5 = RigVMFunctionReferenceNode(
    Rotation Order=ZYX
    child=head
    debugRotation=false
    max=-25
    min=0
    onlyTwist=true
    parent=head_null
    parentType=Null
    premult=1
    primAxis=0
    twistAxis=(X=1,Y=0,Z=0)
    ExecuteContext=@RigVMFunction_Sequence.A
   )
   > &RigUnit_ModifyTransforms(
    ItemToModify="((Item=(Type=Bone,Name=\"aimtarget_hair_front\"),Transform=(Rotation=(X=0,Y=0,Z=0,W=-1),Translation=(X=0,Y=0,Z=0),Scale3D=(X=1,Y=1,Z=1))),(Item=(Type=Bone,Name=\"aimtarget_hair_mid\"),Transform=(Rotation=(X=0,Y=0,Z=0,W=1),Translation=(X=0,Y=0,Z=-3),Scale3D=(X= ...(truncated)"
    WeightMinimum=0
    Weight=@DeltaRotation_5.outputWeight
   )
   > &RigUnit_AimConstraintLocalSpaceOffset(
    Child=(Type=Bone,Name="base_hair_mid")
    Filter=(bX=True,bY=True,bZ=True)
    Parents=((Item=(Type=Bone,Name="base_strap_neck_l"),Weight=1))
    UpAxis=(X=0,Y=-1,Z=0)
    WorldUp=(Target=(X=0,Y=-1,Z=0),Kind=Location,Space=(Type=Bone,Name="base_strap_neck_l"))
    Weight=@DeltaRotation_5.outputWeight
   )
   > &RigUnit_AimConstraintLocalSpaceOffset(
    Child=(Type=Bone,Name="base_hair_back")
    Filter=(bX=True,bY=True,bZ=True)
    Parents=((Item=(Type=Bone,Name="aimtarget_hair_back"),Weight=1))
    UpAxis=(X=0,Y=-1,Z=0)
    Weight=1
    WorldUp=(Target=(X=0,Y=-1,Z=0),Kind=Location,Space=(Type=Bone,Name="aimtarget_hair_back"))
   )
   > &RigUnit_ModifyTransforms(
    ItemToModify=((Item=(Type=Bone,Name="dyn_hair_back"),Transform=(Rotation=(X=0.0076,Y=0.0868,Z=-0.0868,W=0.9924),Translation=(X=0,Y=0,Z=0),Scale3D=(X=0.6,Y=0.6,Z=0.6))))
    WeightMinimum=0
    Weight=@DeltaRotation_5.outputWeight
   )
  ]}
]
hierarchy [
 root {loc=<0 0 0>, rot=<0 0 0 1>}
 pelvis {parent=root, loc=<0 -0.5301 98.0571>, rot=<0.0127 0.707 -0.0127 -0.707>}
 spine_01 {parent=pelvis, loc=<6.1208 -0.9328 0>, rot=<0 0 0.034 -0.9994>}
 spine_02 {parent=spine_01, loc=<6.5997 -0.2095 0>, rot=<0 0 0.0238 -0.9997>}
 spine_03 {parent=spine_02, loc=<6.4914 -0.1237 0>, rot=<0 0 0.0149 -0.9999>}
 spine_04 {parent=spine_03, loc=<6.9598 0.0794 0>, rot=<0 0 -0.0132 -0.9999>}
 spine_05 {parent=spine_04, loc=<6.9987 0 0>, rot=<0 0 -0.0015 -1>}
 neck_01 {parent=spine_05, loc=<11.3138 0.1529 0>, rot=<0 0 -0.2003 -0.9797>}
 neck_02 {parent=neck_01, loc=<4.5627 0 0>, rot=<0 0 0.0504 -0.9987>}
 head {parent=neck_02, loc=<5.3914 0 0>, rot=<0 0 0.115 -0.9934>}
 base_hair_back {parent=head, loc=<-0.2787 2.2535 -7.8514>, rot=<0.0572 -0.9901 -0.1065 0.0708>}
 dyn_hair_back {parent=base_hair_back, loc=<0 0 0>, rot=<0 0 0 1>}
 dyn_hair_back_front {parent=dyn_hair_back, loc=<1.7037 0.6162 0.0151>, rot=<0.0008 0.135 -0.1774 -0.9748>}
 dyn_hair_back_back {parent=dyn_hair_back, loc=<1.7474 -0.5198 0.02>, rot=<-0.1122 0.0225 0.0066 -0.9934>}
 base_hair_mid {parent=head, loc=<-0.0896 5.647 -7.1826>, rot=<-0.0117 -0.9799 -0.1988 -0.0106>}
 dyn_hair_mid_01 {parent=base_hair_mid, loc=<0 0 0>, rot=<0 0 0 1>}
 dyn_hair_mid_02 {parent=dyn_hair_mid_01, loc=<6.3693 -0.1738 -0.4137>, rot=<0.1235 0.178 -0.042 -0.9753>}
 base_hair_front {parent=head, loc=<3.1018 9.4376 -5.2569>, rot=<0.0778 -0.9775 -0.1932 0.0343>}
 dyn_hair_front_01 {parent=base_hair_front, loc=<0 0 0>, rot=<0 0 0 1>}
 dyn_hair_front_02 {parent=dyn_hair_front_01, loc=<6.5253 0 0>, rot=<0.0239 0.0511 -0.053 -0.997>}
 dyn_hair_front_03 {parent=dyn_hair_front_02, loc=<5.2735 0 0>, rot=<-0.1017 0.1773 -0.1348 -0.9696>}
 aimtarget_hair_back {parent=head, loc=<-10.3501 0.9395 -6.5401>, rot=<-0.0572 0.9901 0.1065 -0.0708>}
 clavicle_l {parent=spine_05, loc=<5.4831 -0.1053 -3.2946>, rot=<0.0465 -0.7289 -0.009 -0.6829>}
 upperarm_l {parent=clavicle_l, loc=<9.8658 -1.2136 -0.4105>, rot=<0.0164 -0.3245 -0.0269 -0.9454>}
 clavicle_r {parent=spine_05, loc=<5.4831 -0.1053 3.2946>, rot=<0.7289 0.0465 -0.6829 0.009>}
 upperarm_r {parent=clavicle_r, loc=<-9.8658 1.2136 0.4105>, rot=<0.0164 -0.3245 -0.0269 -0.9454>}
 base_neck_r {parent=spine_05, loc=<11.4331 2.8701 6.8095>, rot=<-0.2327 0.2138 0.0917 -0.9443>}
 dyn_neck_r {parent=base_neck_r, loc=<0 0 0>, rot=<0 0 0 1>}
 base_neck_mid_r {parent=spine_05, loc=<12.8413 -2.1127 5.2165>, rot=<-0.4962 0.1635 0.0779 -0.8491>}
 dyn_neck_mid_r {parent=base_neck_mid_r, loc=<0 0 0>, rot=<0 0 0 1>}
 base_neck_mid {parent=spine_05, loc=<12.863 -4.5171 0.0057>, rot=<-0.6738 0.1301 0.1186 -0.7177>}
 dyn_neck_mid {parent=base_neck_mid, loc=<0 0 0>, rot=<0 0 0 1>}
 base_neck_mid_l {parent=spine_05, loc=<12.8333 -2.1267 -6.2351>, rot=<0.9315 -0.0843 -0.2161 0.2802>}
 dyn_neck_mid_l {parent=base_neck_mid_l, loc=<0 0 0>, rot=<0 0 0 1>}
 base_neck_l {parent=spine_05, loc=<11.496 2.5127 -6.8707>, rot=<0.958 -0.1784 -0.2232 -0.0243>}
 dyn_neck_l {parent=base_neck_l, loc=<0 0 0>, rot=<0 0 0 1>}
 aimtarget_hair_mid {parent=spine_05, loc=<9.9907 8.7196 -7.3462>, rot=<0.0235 -0.9797 -0.199 -0.0034>}
 aimtarget_hair_front {parent=spine_05, loc=<9.4149 9.8675 -4.7356>, rot=<0.113 -0.974 -0.1918 0.0413>}
 head_null {parent=spine_05, type=Null, loc=<20.6573 3.5489 0>, rot=<0 0 -0.036 -0.9994>}
 base_strap_neck_l {parent=spine_05, loc=<9.9136 7.2651 -5.5208>, rot=<-0.245 -0.9542 -0.148 0.0875>}
 dyn_strap_neck_l_01 {parent=base_strap_neck_l, loc=<0 0 0>, rot=<0 0 0 1>}
 dyn_strap_neck_l_02 {parent=dyn_strap_neck_l_01, loc=<5.1211 0.0039 -0.0165>, rot=<-0.2613 0.2306 0.1058 -0.9313>}
 dyn_strap_neck_l_03 {parent=dyn_strap_neck_l_02, loc=<5.0187 -0.0441 -0.0073>, rot=<-0.079 -0.1032 -0.1482 -0.9804>}
]
