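# Physics asset for the face accessory mesh of the F_MED_Nevermore female character, defining secondary-motion bodies for hair, neck pieces and a left neck strap.
# Small base capsules root dyn_ chains with tight angular limits, mostly loosening toward hair and strap tips; back hair forks in two.
# Head, left clavicle and upper arm add larger body capsules, with free-rotation joints linking hair bases to head.
# Only left-side neck pieces have dyn_ bodies; right and center have unconstrained bases.
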PhysicsAsset /BRCosmetics/Characters/Player/Female/Medium/Bodies/F_MED_Nevermore/Meshes/Parts/F_MED_Nevermore_FaceAcc_Physics
bodies {
 base_hair_back [{shapeType=Capsule, length=0.5, radius=0.5, shapeName=base_hair_back_capsule}]
 base_hair_front [{shapeType=Capsule, length=0.5, radius=0.5, shapeName=base_hair_front_capsule}]
 base_hair_mid [{shapeType=Capsule, length=0.5, radius=0.5, shapeName=base_hair_mid_capsule}]
 base_neck_l [{shapeType=Capsule, length=0.505, radius=0.505, shapeName=base_neck_l_capsule, rotation=<rot 0 -90 0>}]
 base_neck_mid [{shapeType=Capsule, length=0.505, radius=0.505, shapeName=base_neck_mid_capsule, rotation=<rot 0 -90 0>}]
 base_neck_mid_l [{shapeType=Capsule, length=0.505, radius=0.505, shapeName=base_neck_mid_l_capsule, rotation=<rot 0 -90 0>}]
 base_neck_mid_r [{shapeType=Capsule, length=0.505, radius=0.505, shapeName=base_neck_mid_r_capsule, rotation=<rot 0 -90 0>}]
 base_neck_r [{shapeType=Capsule, length=0.505, radius=0.505, shapeName=base_neck_r_capsule, rotation=<rot 0 -90 0>}]
 base_strap_neck_l [{shapeType=Capsule, length=0.505, radius=0.505, shapeName=base_strap_neck_l_capsule, rotation=<rot 0 -90 0>}]
 clavicle_l [{shapeType=Capsule, length=9, radius=6, shapeName=clavicle_l_capsule, center=<1.5 1 -3>, rotation=<rot 90 0 0>}]
 dyn_hair_back [{shapeType=Capsule, length=2, radius=0.6, shapeName=dyn_hair_back_capsule, center=<0.5 0 0.8>, rotation=<rot 90 0 0>}]
 dyn_hair_back_back [{shapeType=Capsule, length=1.5, radius=0.6, shapeName=dyn_hair_back_back_capsule, center=<1.3873 0 0.4>, rotation=<rot 70 0.0002 0>}]
 dyn_hair_back_front [{shapeType=Capsule, length=4, radius=0.8, shapeName=dyn_hair_back_front_capsule, center=<3 0 0.7>, rotation=<rot 90 0 0>}]
 dyn_hair_front_01 [{
   shapeType=Capsule
   length=4
   radius=1.8
   shapeName=dyn_hair_front_01_capsule
   center=<2.614 0.15 1.3>
   rotation=<rot 79.9999 -174.9998 179.9999>
  }]
 dyn_hair_front_02 [{shapeType=Capsule, length=4, radius=1.8, shapeName=dyn_hair_front_02_capsule, center=<2.5542 0.3 1>, rotation=<rot 90 5.0001 0>}]
 dyn_hair_front_03 [{shapeType=Capsule, length=6, radius=1.5, shapeName=dyn_hair_front_03_capsule, center=<3.5 0 0>, rotation=<rot 90 0 0>}]
 dyn_hair_mid_01 [{shapeType=Capsule, length=4, radius=1.8, shapeName=dyn_hair_mid_01_capsule, center=<2.5978 -0.5 1>, rotation=<rot 90 -4.9999 0>}]
 dyn_hair_mid_02 [{shapeType=Capsule, length=4, radius=1.5, shapeName=dyn_hair_mid_02_capsule, center=<1.9825 -0.5 0.85>, rotation=<rot 90 0 0>}]
 dyn_neck_l [{shapeType=Capsule, length=14, radius=5, shapeName=dyn_neck_l_capsule, center=<9.1 0.5 1.9>, rotation=<rot 90 0.0002 0>}]
 dyn_neck_mid_l [{shapeType=Capsule, length=9, radius=5.2, shapeName=dyn_neck_mid_l_capsule, center=<5.0042 0 2.5>, rotation=<rot 85 0 0>}]
 dyn_strap_neck_l_01 [{shapeType=Capsule, length=7, radius=1.2, shapeName=dyn_strap_neck_l_01_capsule, center=<2 0 0>, rotation=<rot 90 0 0>}]
 dyn_strap_neck_l_02 [{shapeType=Capsule, length=4, radius=1.3, shapeName=dyn_strap_neck_l_02_capsule, center=<2.4939 0.1 0>, rotation=<rot 90 0 0>}]
 dyn_strap_neck_l_03 [{shapeType=Capsule, length=4, radius=1.3, shapeName=dyn_strap_neck_l_03_capsule, center=<2.4939 0.5 0>, rotation=<rot 90 0 0>}]
 head [{shapeType=Capsule, length=8, radius=8, shapeName=head_capsule, center=<3.0034 0.3919 0>, rotation=<rot 90 -25 0>}]
 upperarm_l [{shapeType=Capsule, length=11, radius=4, shapeName=upperarm_l_capsule, center=<8.1856 0 -1.3012>, rotation=<rot 90 0 0>}]
}
constraints [
 {bone1Name=dyn_neck_l, bone2Name=base_neck_l, swing1LimitDegrees=30, swing2LimitDegrees=10, twistLimitDegrees=5}
 {bone1Name=dyn_neck_mid_l, bone2Name=base_neck_mid_l, swing1LimitDegrees=25, swing2LimitDegrees=10, twistLimitDegrees=5}
 {
  bone1Name=upperarm_l
  bone2Name=clavicle_l
  swing1LimitDegrees=45
  swing1Motion=Free
  swing2LimitDegrees=45
  swing2Motion=Free
  twistLimitDegrees=45
  twistMotion=Free
 }
 {
  bone1Name=base_hair_back
  bone2Name=head
  swing1LimitDegrees=45
  swing1Motion=Free
  swing2LimitDegrees=45
  swing2Motion=Free
  twistLimitDegrees=45
  twistMotion=Free
 }
 {bone1Name=dyn_hair_back, bone2Name=base_hair_back, swing1LimitDegrees=5, swing2LimitDegrees=10, twistLimitDegrees=1}
 {bone1Name=dyn_hair_back_back, bone2Name=dyn_hair_back, swing1LimitDegrees=5, swing2LimitDegrees=15, twistLimitDegrees=1}
 {bone1Name=dyn_hair_back_front, bone2Name=dyn_hair_back, swing1LimitDegrees=5, swing2LimitDegrees=15, twistLimitDegrees=1}
 {
  bone1Name=base_hair_front
  bone2Name=head
  swing1LimitDegrees=45
  swing1Motion=Free
  swing2LimitDegrees=45
  swing2Motion=Free
  twistLimitDegrees=45
  twistMotion=Free
 }
 {bone1Name=dyn_hair_front_01, bone2Name=base_hair_front, swing1LimitDegrees=15, swing2LimitDegrees=10, twistLimitDegrees=1}
 {bone1Name=dyn_hair_front_02, bone2Name=dyn_hair_front_01, swing1LimitDegrees=20, swing2LimitDegrees=20, twistLimitDegrees=1}
 {bone1Name=dyn_hair_front_03, bone2Name=dyn_hair_front_02, swing1LimitDegrees=30, swing2LimitDegrees=25, twistLimitDegrees=1}
 {
  bone1Name=base_hair_mid
  bone2Name=head
  swing1LimitDegrees=45
  swing1Motion=Free
  swing2LimitDegrees=45
  swing2Motion=Free
  twistLimitDegrees=45
  twistMotion=Free
 }
 {bone1Name=dyn_hair_mid_01, bone2Name=base_hair_mid, swing1LimitDegrees=5, swing2LimitDegrees=15, twistLimitDegrees=1}
 {bone1Name=dyn_hair_mid_02, bone2Name=dyn_hair_mid_01, swing1LimitDegrees=10, swing2LimitDegrees=20, twistLimitDegrees=1}
 {bone1Name=dyn_strap_neck_l_03, bone2Name=dyn_strap_neck_l_02, swing1LimitDegrees=20, swing2LimitDegrees=10, twistLimitDegrees=5}
 {bone1Name=dyn_strap_neck_l_01, bone2Name=base_strap_neck_l, swing1LimitDegrees=15, swing2LimitDegrees=8, twistLimitDegrees=5}
 {bone1Name=dyn_strap_neck_l_02, bone2Name=dyn_strap_neck_l_01, swing1LimitDegrees=18, swing2LimitDegrees=10, twistLimitDegrees=5}
]
